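# Niagara laser beam effect for the South Bulk boss's energy ray, with a glowing muzzle.
# Beam lays six particles from origin to impact using shared direction and distance variables, drawing layered ribbons with a pulsing core width.
# MuzzleCore adds flickering red sprites and a light after a spawn delay, Minimal005 a brief glow, and LaserFadeOut scales the beam and muzzle.
# Many secondary emitters, including impact, rocks, scorch trail and shockwave, are disabled.
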NiagaraSystem /SouthBulkAI/Assets/VFX/Laser/NS_SouthBulk_Boss_EnergyRay
dependencies {
 usedDataInterfaces [
  &NiagaraDataInterfaceCurve
  &NiagaraDataInterfaceSpriteRendererInfo
  &NiagaraDataInterfaceColorCurve
 ]
 usedDynamicInputs [
  &MultiplyQuaternion
  &MakeVector4_FromLinearColor
  &LerpLinearColors
  &MakeFloatFromVector
 ]
 usedModules [
  &SystemState
  &self:SystemSpawnScript.SystemScriptSource.SystemScriptGraph.NiagaraNodeAssignment_0.SetVariables_#1
  &self:SystemSpawnScript.SystemScriptSource.SystemScriptGraph.NiagaraNodeAssignment_6.SetVariables_#2
  &self:SystemSpawnScript.SystemScriptSource.SystemScriptGraph.NiagaraNodeAssignment_1.SetVariables_#3
  &self:SystemSpawnScript.SystemScriptSource.SystemScriptGraph.NiagaraNodeAssignment_2.SetVariables_#4
  &self:SystemSpawnScript.SystemScriptSource.SystemScriptGraph.NiagaraNodeAssignment_3.SetVariables_#5
  &self:SystemSpawnScript.SystemScriptSource.SystemScriptGraph.NiagaraNodeAssignment_9.SetVariables_#6
  &self:SystemSpawnScript.SystemScriptSource.SystemScriptGraph.NiagaraNodeAssignment_4.SetVariables_#7
  &self:SystemSpawnScript.SystemScriptSource.SystemScriptGraph.NiagaraNodeAssignment_5.SetVariables_#8
 ]
}
emitters {
 Beam {
  emitterUpdateScript [
   &EmitterState(
    Life Cycle Mode=::NewEnumerator1
    Loop Behavior=::NewEnumerator1
    Loop Delay=&Add_Float(A=$System.BeamDelaySpawn, B=0.05)
    Loop Duration Mode=::NewEnumerator1
    UseLoopDelay=true
   )
   &SpawnBurst_Instantaneous(Spawn Count=6, Use Spawn Probability=false, UseLoopCountLimit=false)
  ]
  particleSpawnScript [&InitializeParticle(Lifetime=0.6, Ribbon Width=70, Ribbon Width Mode=::NewEnumerator1, UsePositionOffset=false, Write Material Random=true)]
  particleUpdateScript [
   &ParticleState(Kill Particles When Lifetime Has Elapsed=false, Let Infinitely Lived Particles Die When Emitter Deactivates=true)
   /Game/Effects/Niagara/Modules/Curves/FN_HermiteSpline(
    Position Blend:NiagaraFloat=&OneMinusFloat(Float=&ReturnNormalizedExecIndex())
    Position End:Vector3f=&Multiply_VectorByFloat(Vector=$System.BeamDirection, Float=&Add_Float(A=$System.BeamImpactDistance, B=55))
    Position Start:Vector3f=<0 0 0>
    Write Particles Position=true
   ) {enabled=false}
   /SetVariables(Particles.Position=&Multiply_VectorByFloat(
     Vector=&Multiply_VectorByFloat(Vector=$System.BeamDirection, Float=&Add_Float(A=$System.BeamImpactDistance, B=55))
     Float=&OneMinusFloat(Float=&ReturnNormalizedExecIndex())
    )) {isSetParameters=true}
   &ScaleRibbonWidth(Ribbon Width Scale=&Multiply_Float(
     A=&FloatFromCurve(FloatCurve=DataInterface() {propertyValues {Curve {keys [
          {arriveTangent=5.1162, interpMode=RCIM_Cubic, leaveTangent=5.1162, tangentMode=RCTM_User, time=0, value=0.1}
          {interpMode=RCIM_Cubic, tangentMode=RCTM_User, time=0.2, value=1}
         ]}}})
     B=$System.LaserFadeOut
    ))
   /SetVariables(
    Particles.ColorGlow:LinearColor=&MultiplyLinearColor(
     Linear Color=<rgba 0.4 0.32 0.3253 0.4>
     Scale Factor=&MakeVector4FromVectorAndFloat(
      "Vector (RGB)"=<1 1 1>
      "Float (Alpha)"=&FloatFromCurve(
       FloatCurve=DataInterface() {propertyValues {Curve {keys [
           {arriveTangent=-0.2002, interpMode=RCIM_Cubic, leaveTangent=-0.2002, tangentMode=RCTM_User, time=0.9, value=1}
           {arriveTangent=-10.8263, interpMode=RCIM_Cubic, leaveTangent=-10.8263, tangentMode=RCTM_User, time=1, value=0.5}
          ]}}}
       CurveIndex=&ReturnNormalizedExecIndex()
      )
     )
    )
    Particles.WidthGlow:NiagaraFloat=&Multiply_Float(A=$Particles.RibbonWidth, B=1.7)
   ) {isSetParameters=true}
   /SetVariables(
    Particles.ColorCore:LinearColor=&MultiplyLinearColor(
     Linear Color=<rgba 2 2 2 1>
     Scale Factor=&MakeVector4FromVectorAndFloat("Vector (RGB)"=<1 1 1>, "Float (Alpha)"=1)
    )
    Particles.WidthCore:NiagaraFloat=&Multiply_Float(
     A=$Particles.RibbonWidth
     B=&Multiply_Float(
      A=0.5
      B=&FloatFromCurve(
       FloatCurve=DataInterface() {propertyValues {Curve {keys [<at 0 1>,<at 0.1 1.45>,<at 0.2 0.5>,<at 0.5 0.95>,<at 0.7 1.15>,<at 0.8 0.3>,<at 1 1>]}}}
       CurveIndex=&ModuloFloat(A=&Multiply_Float(A=$Particles.Age, B=2), B=1)
      )
     )
    )
   ) {isSetParameters=true}
   /SetVariables(
    Particles.ColorMain:LinearColor=&MultiplyLinearColor(
     Linear Color=<rgba 1 1 1 0.5>
     Scale Factor=&MakeVector4FromVectorAndFloat(
      "Vector (RGB)"=<1 1 1>
      "Float (Alpha)"=&FloatFromCurve(
       FloatCurve=DataInterface() {propertyValues {Curve {keys [
           {arriveTangent=-0.2002, interpMode=RCIM_Cubic, leaveTangent=-0.2002, tangentMode=RCTM_User, time=0.9, value=1}
           {arriveTangent=-10.8263, interpMode=RCIM_Cubic, leaveTangent=-10.8263, tangentMode=RCTM_User, time=1, value=0.5}
          ]}}}
       CurveIndex=&ReturnNormalizedExecIndex()
      )
     )
    )
    Particles.WidthMain:NiagaraFloat=&Multiply_Float(A=$Particles.RibbonWidth, B=1.15)
   ) {isSetParameters=true}
  ]
  rendererClasses [&NiagaraRibbonRendererProperties, &NiagaraSpriteRendererProperties]
  renderers [
   {rendererClass=&NiagaraRibbonRendererProperties}
   {rendererClass=&NiagaraRibbonRendererProperties, rendererIndex=1}
   {rendererClass=&NiagaraRibbonRendererProperties, rendererIndex=2}
   {rendererClass=&NiagaraSpriteRendererProperties, rendererIndex=3}
  ]
 }
 EnergyThorn {
  bEnabled=false
  emitterUpdateScript [
   &EmitterState(Life Cycle Mode=::NewEnumerator1, Loop Behavior=::NewEnumerator1, Loop Duration Mode=::NewEnumerator1, UseLoopDelay=true)
   &SpawnBurst_Instantaneous(Spawn Count=3, Use Spawn Probability=false, UseLoopCountLimit=false) {enabled=false}
   &SpawnRate(Spawn Probability=$System.LaserFadeOut, SpawnRate=30, Use Spawn Probability=true)
  ]
  particleSpawnScript [
   &InitializeParticle(
    Color Maximum=<rgba 2 0.1 0.1 1>
    Color Minimum=<rgba 1.5 0.03 0.079 1>
    Color Mode=::NewEnumerator2
    Lifetime Max=0.2
    Lifetime Min=0.15
    Lifetime Mode=::NewEnumerator1
    Mesh Scale Max=<3.5 3.5 3.5>
    Mesh Scale Min=<2.5 2.5 2.5>
    Mesh Scale Mode=::NewEnumerator2
    UsePositionOffset=false
    Write Material Random=true
   )
   &AddVelocity(Module.Use Falloff Distance=false, Override Seed=false, Velocity Mode=::NewEnumerator1, Velocity Speed=1)
   &InitialMeshOrientation(
    Mesh Orientation Mode=::NewEnumerator2
    Rotation=&RandomRangeVector(Minimum=<0 0 0>, Maximum=<1 0 0>)
    Use Rotation Vector=true
   )
  ]
  particleUpdateScript [
   &SolveForcesAndVelocity(Clamp Velocity=false, Limit Acceleration=false)
   &ParticleState()
   &ScaleMeshSize(Scale Factor=&Multiply_Vector(
     A=&MakeVector(
      X=&FloatFromCurve(FloatCurve=DataInterface() {propertyValues {Curve {keys [<at 0 0.2>,<at 1 1>]}}})
      Y=&FloatFromCurve(FloatCurve=DataInterface() {propertyValues {Curve {keys [{interpMode=RCIM_Cubic, tangentMode=RCTM_User, time=0, value=0.5}, {interpMode=RCIM_Cubic, tangentMode=RCTM_User, time=1, value=2}]}}})
      Z=&FloatFromCurve(FloatCurve=DataInterface() {propertyValues {Curve {keys [{interpMode=RCIM_Cubic, tangentMode=RCTM_User, time=0, value=0.5}, {interpMode=RCIM_Cubic, tangentMode=RCTM_User, time=1, value=2}]}}})
     )
     B=<1 1 1>
    ))
   &ScaleColor(
    Scale Alpha=&FloatFromCurve(
     FloatCurve=DataInterface() {propertyValues {Curve {keys [
         {arriveTangent=2.8247, interpMode=RCIM_Cubic, leaveTangent=2.8247, tangentMode=RCTM_User, time=0, value=0}
         {arriveTangent=0.0384, time=0.1, value=1}
         {arriveTangent=-1.1111, interpMode=RCIM_Cubic, leaveTangent=-1.1111, time=0.5, value=1}
         {arriveTangent=0.0424, interpMode=RCIM_Cubic, leaveTangent=0.0424, tangentMode=RCTM_User, time=1, value=0}
        ]}}}
     Scale Curve=&Multiply_Float(A=0.8, B=$System.LaserFadeOut)
    )
    Scale RGB=&MakeVectorFromLinearColorRGB(Color=&ColorFromCurve(ColorCurve=DataInterface() {propertyValues {
        AlphaCurve {keys []}
        BlueCurve {keys [{arriveTangent=-6.2021, time=0.1407, value=2}, {arriveTangent=-1.9715, time=0.4613, value=1}, {leaveTangent=-0.0154, time=0.9275, value=0.1}]}
        GreenCurve {keys [
          {arriveTangent=-6.099, time=0.1407, value=2}
          {arriveTangent=0.1945, time=0.4613, value=1}
          {arriveTangent=-0.0129, leaveTangent=-0.0129, time=0.9275, value=0.1}
         ]}
        RedCurve {keys [
          {arriveTangent=-31.4411, time=0.1407, value=2}
          {arriveTangent=-21.5603, time=0.4613, value=1}
          {leaveTangent=-0.0154, time=0.9275, value=0.1}
         ]}
       }}))
    ScaleA=true
    ScaleRGB=true
   )
  ]
  rendererClasses [&NiagaraMeshRendererProperties]
  renderers [{rendererClass=&NiagaraMeshRendererProperties, rendererIndex=0}]
 }
 EnergyThornLW {rendererClasses [], renderers []}
 ImpactCore {
  bEnabled=false
  emitterUpdateScript [
   &EmitterState(
    Life Cycle Mode=::NewEnumerator1
    Loop Behavior=::NewEnumerator1
    Loop Delay=&Add_Float(A=$System.BeamDelaySpawn, B=0.15)
    Loop Duration Mode=::NewEnumerator1
    UseLoopDelay=true
   )
   &SpawnBurst_Instantaneous(Use Spawn Probability=false, UseLoopCountLimit=false)
   /SetVariables(Emitter.FlickerAnimAlpha:NiagaraFloat=&Waveform(
     "Waveform [1]"=::NewEnumerator6
     "[1] Input Position"=&RandomRangeFloat(Minimum=0, Maximum=1)
     "[1] Frequency"=7
     "[1] Amplitude Min/Max"=<0 1>
    )) {isSetParameters=true, module=SetVariables_#9}
  ]
  particleSpawnScript [
   &InitializeParticle(
    Color=<rgba 6 0 0 1.5>
    Mesh Scale Mode=::NewEnumerator3
    Mesh Uniform Scale=0.4
    Position Offset:NiagaraPosition=$System.ImpactLocation
    Sprite Size Mode=::NewEnumerator3
    Uniform Sprite Size=400
    UsePositionOffset=true
    Write Material Random=true
   )
   &InitialMeshOrientation(Mesh Axis To Orient=<0 1 0>, Mesh Orientation Mode=::NewEnumerator2, Use Rotation Vector=false)
  ]
  particleUpdateScript [
   &ParticleState(Kill Particles When Lifetime Has Elapsed=false, Let Infinitely Lived Particles Die When Emitter Deactivates=true)
   &ScaleMeshSize(Scale Factor=&VectorFromFloat(Value=&Multiply_Float(A=&Lerp_Float(A=0.8, B=1.3, Alpha=$Emitter.FlickerAnimAlpha), B=$System.LaserFadeOut)))
   &ScaleSpriteSize(
    Scale Sprite Size Mode=::NewEnumerator0
    Uniform Scale Factor=&Multiply_Float(A=&Lerp_Float(A=0.7, B=1.1, Alpha=$Emitter.FlickerAnimAlpha), B=$System.LaserFadeOut)
   )
   /SetVariables(
    Particles.ColorGlow:LinearColor=<rgba 1 0 0 1>
    Particles.SpriteSizeGlow:Vector2f=&Multiply_Vector2D_ByFloat(Float=1, Vector2D=$Particles.SpriteSize)
   ) {isSetParameters=true, module=SetVariables_#10}
   /SetVariables(
    Particles.ColorCore:LinearColor=<rgba 1 0.02 0.0527 2>
    Particles.SpriteSizeCore:Vector2f=&Multiply_Vector2D_ByFloat(Float=0.3, Vector2D=$Particles.SpriteSize)
   ) {isSetParameters=true, module=SetVariables_#11}
   /SetVariables(
    Particles.ColorCoreSmall:LinearColor=<rgba 2 0.5067 0.4 1>
    Particles.SpriteSizeCoreSmall:Vector2f=&Multiply_Vector2D_ByFloat(Float=0.5, Vector2D=$Particles.SpriteSize)
   ) {isSetParameters=true, module=SetVariables_#12}
   /SetVariables(
    Particles.ColorFlare:LinearColor=<rgba 5 0 0 0.2>
    Particles.SpriteSizeFlare:Vector2f=&Multiply_Vector2D_ByFloat(Float=0.55, Vector2D=$Particles.SpriteSize)
   ) {isSetParameters=true, module=SetVariables_#13}
   /SetVariables(Particles.ScaleGlow:Vector3f=&Multiply_VectorByFloat(Vector=$Particles.Scale, Float=1.4)) {isSetParameters=true, module=SetVariables_#14}
   &Light_Attributes(
    Color=1
    Falloff Exponent=2
    Light Exponent=6
    Light Radius=300
    Radius=2
    Scale Alpha:NiagaraFloat=&Multiply_Float(A=&Lerp_Float(A=2, B=7, Alpha=$Emitter.FlickerAnimAlpha), B=$System.LaserFadeOut)
    ScaleA=true
    ScaleRGB=true
   )
   &CameraOffset(Camera Offset Amount=30)
  ]
  rendererClasses [&NiagaraSpriteRendererProperties, &NiagaraLightRendererProperties]
  renderers [
   {rendererClass=&NiagaraSpriteRendererProperties}
   {rendererClass=&NiagaraSpriteRendererProperties, rendererIndex=1}
   {rendererClass=&NiagaraLightRendererProperties, rendererIndex=2}
   {rendererClass=&NiagaraSpriteRendererProperties, rendererIndex=3}
   {rendererClass=&NiagaraSpriteRendererProperties, rendererIndex=4}
  ]
 }
 ImpactFlashes {
  bEnabled=false
  emitterUpdateScript [
   &EmitterState(
    Life Cycle Mode=::NewEnumerator1
    Loop Behavior=::NewEnumerator1
    Loop Delay=&Add_Float(A=$System.BeamDelaySpawn, B=0.15)
    Loop Duration Mode=::NewEnumerator1
    UseLoopDelay=true
   )
   &SpawnRate(Spawn Probability=$System.LaserFadeOut, SpawnRate=20, Use Spawn Probability=true)
  ]
  particleSpawnScript [
   &InitializeParticle(
    Color=<rgba 2 2 2 0.5>
    Lifetime=&RandomRangeFloat(Minimum=0.27, Maximum=0.4)
    Sprite Size=&Multiply_Vector2D_ByFloat(
     Float=&RandomRangeFloat(Minimum=2, Maximum=3)
     Vector2D=&RandomRangeVector2D(Minimum=<50 200>, Maximum=<80 220>)
    )
    Sprite Size Mode=::NewEnumerator4
    UsePositionOffset=false
    Write Material Random=true
   )
   &InitialMeshOrientation(
    Mesh Axis To Orient=<-1 0 0>
    Mesh Orientation Mode=::NewEnumerator2
    Orientation Vector=$System.BeamDirection
    Rotation=&RandomRangeVector(Minimum=<-1 0 0>, Maximum=<1 0 0>)
    Use Rotation Vector=true
   )
   &SubUVAnimation(
    "Sprite Renderer "=DataInterface() {propertyValues {SpriteRenderer=&self:ImpactFlashes_0.NiagaraSpriteRendererProperties_1}}
    SubUV Animation Mode:ENiagara_SubUVLookupModeV2=::NewEnumerator2
    UseEndFrame=false
    UseStartFrame=false
   )
   &ShapeLocation(
    MaskSpawnGroup=false
    Offset:NiagaraPosition=$System.ImpactLocation
    Offset Coordinate Space=::Local
    Offset Mode=::NewEnumerator0
    Override Seed=false
    Rotation Coordinate Space=::Simulation
    Sphere Radius=6
    UseOwnerScale=true
   )
   &SpriteFacingAndAlignment(
    Sprite Alignment=&Lerp_Vector(A=$System.BeamDirection, B=$Particles.ShapeLocation.ShapeTangent, Alpha=0.7)
    WriteAlignment=true
    WriteFacing=true
   ) {enabled=false}
   &AddVelocity(
    Module.Use Falloff Distance=false
    Origin Offset=&Multiply_VectorByFloat(
     Vector=&AddVectorToPosition(
      Position=Linked() {linkedVariable {name=System.BeamDirection, type=Vector3f}}
      Vector=&RandomRangeVector(Minimum=<0.02 -0.1 -0.1>, Maximum=<0.1 0.1 -0.2>)
     )
     Float=380
    )
    Override Seed=false
    Velocity Mode=::NewEnumerator1
    Velocity Speed=&RandomRangeFloat(Minimum=10, Maximum=20)
   )
   &CameraOffset(Camera Offset Amount=20)
  ]
  particleUpdateScript [
   &ParticleState()
   &ScaleSpriteSize(Uniform Curve Sprite Scale=DataInterface() {propertyValues {Curve {keys [{tangentMode=RCTM_User, time=0, value=0.4}, {tangentMode=RCTM_User, time=1, value=1}]}}})
   &DynamicMaterialParameters(
    Index 0 Param 1=20
    Index 0 Param 2=1
    Index 0 Param 3=1
    Index 0 Param 4=&FloatFromCurve(FloatCurve=DataInterface() {propertyValues {Curve {keys [
         {interpMode=RCIM_Cubic, tangentMode=RCTM_User, time=0, value=0}
         {arriveTangent=1.0951, interpMode=RCIM_Cubic, leaveTangent=1.0951, tangentMode=RCTM_User, time=1, value=1}
        ]}}})
    Index 2 Param 1=0.5
    Index 2 Param 2=&FloatFromCurve(FloatCurve=DataInterface() {propertyValues {Curve {keys [
         {arriveTangent=-3.0317, interpMode=RCIM_Cubic, leaveTangent=-3.0317, tangentMode=RCTM_User, time=0, value=1}
         {arriveTangent=-0.1613, interpMode=RCIM_Cubic, leaveTangent=-0.1613, tangentMode=RCTM_User, time=1, value=0}
        ]}}})
    Module.Param0WriteEnabled=true
    Module.Param0WriteEnabled2=true
    Module.Param1WriteEnabled=false
    Module.Param1WriteEnabled2=true
    Module.Param2WriteEnabled=false
    Module.Param2WriteEnabled2=true
    Module.Param3WriteEnabled=true
    Module.Param3WriteEnabled2=true
    Write Parameter Index 2=true
   )
   &ScaleColor(
    Scale Alpha=&FloatFromCurve(FloatCurve=DataInterface() {propertyValues {Curve {keys [
         {interpMode=RCIM_Cubic, tangentMode=RCTM_User, time=0, value=0}
         {interpMode=RCIM_Cubic, tangentMode=RCTM_User, time=0.15, value=1}
         {arriveTangent=-3.1002, interpMode=RCIM_Cubic, leaveTangent=-3.1002, tangentMode=RCTM_User, time=1, value=0}
        ]}}})
    Scale RGB=&MakeVectorFromLinearColorRGB(Color=&ColorFromCurve(ColorCurve=DataInterface() {propertyValues {
        AlphaCurve {keys [<at 0.1519 1 RCIM_Cubic>]}
        BlueCurve {keys [<at 0.0008 0.069 RCIM_Cubic>,<at 0.3233 0 RCIM_Cubic>]}
        GreenCurve {keys [<at 0.0008 0.02 RCIM_Cubic>,<at 0.3233 0 RCIM_Cubic>]}
        RedCurve {keys [<at 0.0008 1 RCIM_Cubic>,<at 0.3233 1 RCIM_Cubic>]}
       }}))
    ScaleA=true
    ScaleRGB=true
   )
   /SetVariables(
    Particles.ColorCore:LinearColor=&ColorFromCurve(ColorCurve=DataInterface() {propertyValues {
       AlphaCurve {keys [<at 0.0008 0>,<at 0.0997 0.55 RCIM_Cubic>,<at 1 0 RCIM_Cubic>]}
       BlueCurve {keys [<at 0.0566 1 RCIM_Cubic>,<at 0.3246 0.1906>,<at 0.7329 0.0083 RCIM_Cubic>]}
       GreenCurve {keys [<at 0.0566 1 RCIM_Cubic>,<at 0.3246 0.1887>,<at 0.7329 0 RCIM_Cubic>]}
       RedCurve {keys [{arriveTangent=3.1293, interpMode=RCIM_Cubic, time=0.0566, value=5}, <at 0.3246 2>, <at 0.7329 0.5 RCIM_Cubic>]}
      }})
    Particles.SpriteSizeCore:Vector2f=&Multiply_Vector2D(A=$Particles.SpriteSize, B=<0.66 0.8>)
   ) {isSetParameters=true, module=SetVariables_#15}
   &SolveForcesAndVelocity(Clamp Velocity=false, Limit Acceleration=false)
  ]
  rendererClasses [&NiagaraSpriteRendererProperties]
  renderers [
   {rendererClass=&NiagaraSpriteRendererProperties}
   {rendererClass=&NiagaraSpriteRendererProperties, rendererIndex=1}
  ]
 }
 ImpactRocks {
  bEnabled=false
  emitterUpdateScript [
   &EmitterState()
   &SpawnRate(
    Spawn Probability=$System.LaserFadeOut
    SpawnRate=&FloatFromCurve(
     FloatCurve=DataInterface() {propertyValues {Curve {keys [
         {arriveTangent=1.7658, interpMode=RCIM_Cubic, leaveTangent=1.7658, tangentMode=RCTM_User, time=0.3, value=0}
         {interpMode=RCIM_Cubic, tangentMode=RCTM_User, time=0.66, value=1}
        ]}}}
     CurveIndex=$Engine.System.Age
     Scale Curve=66
    )
    Use Spawn Probability=true
   )
  ]
  particleSpawnScript [
   &InitializeParticle(
    Color Mode=::NewEnumerator0
    Lifetime=&RandomRangeFloat(Minimum=0.35, Maximum=0.6)
    Mesh Scale=&Multiply_VectorByFloat(
     Vector=&RandomRangeVector(Minimum=<0.8 0.8 0.8>, Maximum=<1 1 1>)
     Float=&RandomRangeFloat(Minimum=0.1, Maximum=0.4)
    )
    Mesh Scale Mode=::NewEnumerator4
    UsePositionOffset=false
    Write Material Random=true
   )
   &ShapeLocation(
    MaskSpawnGroup=false
    Offset:NiagaraPosition=$System.ImpactLocation
    Offset Coordinate Space=::Local
    Offset Mode=::NewEnumerator0
    Override Seed=false
    Sphere Radius=33
    Sphere Surface Distribution=0.2738
    UseOwnerScale=true
   )
   &AddVelocity(
    Module.Use Falloff Distance=false
    Origin Offset=&AddVectorToPosition(
     Position=&Multiply_VectorByFloat(Vector=Linked() {linkedVariable {name=System.ImpactLocation, type=NiagaraPosition}}, Float=1.15)
     Vector=<0 0 0>
    )
    Override Seed=false
    Velocity Mode=::NewEnumerator1
    Velocity Speed=&RandomRangeFloat(Minimum=700, Maximum=1000)
   )
   &InitialMeshOrientation(Use Rotation Vector=false)
  ]
  particleUpdateScript [
   &ParticleState()
   &ScaleMeshSize(Scale Factor=&VectorFromFloat(Value=&FloatFromCurve(FloatCurve=DataInterface() {propertyValues {Curve {keys [
          {arriveTangent=1.9167, interpMode=RCIM_Cubic, leaveTangent=1.9167, tangentMode=RCTM_User, time=0, value=0}
          {interpMode=RCIM_Cubic, tangentMode=RCTM_User, time=0.22, value=1}
          {interpMode=RCIM_Cubic, tangentMode=RCTM_User, time=0.7, value=1}
          {arriveTangent=-4.0991, interpMode=RCIM_Cubic, leaveTangent=-4.0991, tangentMode=RCTM_User, time=1, value=0}
         ]}}})))
   &UpdateMeshOrientation(
    Rotation Rate=&RandomRangeFloat(Minimum=1, Maximum=2)
    Rotation Vector=&RandomRangeVector(Minimum=<-1 -1 -1>, Maximum=<1 1 1>)
   )
   &Drag(Drag=&RandomRangeFloat(Minimum=0.66, Maximum=1.1), Use Linear Drag=true, Use Rotational Drag=true)
   &GravityForce(Gravity=&Multiply_VectorByFloat(Vector=<0 0 -980>, Float=&RandomRangeFloat(Minimum=1, Maximum=2)))
   &InheritSourceMovement(
    Applied Position Scale=&VectorFromFloat(Value=&FloatFromCurve(FloatCurve=DataInterface() {propertyValues {Curve {keys [
          {arriveTangent=-2.7977, interpMode=RCIM_Cubic, leaveTangent=-2.7977, tangentMode=RCTM_User, time=0, value=1}
          {interpMode=RCIM_Cubic, tangentMode=RCTM_User, time=0.5, value=0}
         ]}}}))
    ApplyForce=false
    ApplyPosition=true
    ApplyPositionFalloff=false
   )
   &SolveForcesAndVelocity(Clamp Velocity=false, Limit Acceleration=false)
  ]
  rendererClasses [&NiagaraMeshRendererProperties]
  renderers [{rendererClass=&NiagaraMeshRendererProperties, rendererIndex=0}]
 }
 Minimal005 {
  emitterUpdateScript [
   &EmitterState(Life Cycle Mode=::NewEnumerator1, Loop Behavior=::NewEnumerator1, Loop Duration=1, UseLoopDelay=false)
   &SpawnBurst_Instantaneous(Use Spawn Probability=false, UseLoopCountLimit=false)
  ]
  particleSpawnScript [&InitializeParticle(
    Color=<rgba 1 1 1 100>
    Lifetime=0.4
    Sprite Size Mode=::NewEnumerator3
    Uniform Sprite Size=300
    UsePositionOffset=false
    Write Material Random=true
   )]
  particleUpdateScript [
   &ParticleState()
   /SetVariables(Particles.ColorGlow:LinearColor=&ColorFromCurve(ColorCurve=DataInterface() {propertyValues {
       AlphaCurve {keys [<at 0 0>,<at 0.2022 1 RCIM_Cubic>,<at 0.4402 1>,<at 1 0 RCIM_Cubic>]}
       BlueCurve {keys [<at 0.0368 0>,<at 0.7902 0.02 RCIM_Cubic>]}
       GreenCurve {keys [<at 0.0368 0>,<at 0.7902 0.02 RCIM_Cubic>]}
       RedCurve {keys [<at 0.0368 0.1>,<at 0.7902 2 RCIM_Cubic>]}
      }})) {isSetParameters=true}
  ]
  rendererClasses [&NiagaraSpriteRendererProperties]
  renderers [{rendererClass=&NiagaraSpriteRendererProperties}]
 }
 MuzzleCore {
  emitterUpdateScript [
   &EmitterState(
    Life Cycle Mode=::NewEnumerator1
    Loop Behavior=::NewEnumerator1
    Loop Delay=$System.BeamDelaySpawn
    Loop Duration Mode=::NewEnumerator1
    UseLoopDelay=true
   )
   &SpawnBurst_Instantaneous(Use Spawn Probability=false, UseLoopCountLimit=false)
   /SetVariables(Emitter.FlickerAnimAlpha:NiagaraFloat=&Waveform(
     "Waveform [1]"=::NewEnumerator6
     "[1] Input Position"=&RandomRangeFloat(Minimum=0, Maximum=1)
     "[1] Frequency"=7
     "[1] Amplitude Min/Max"=<0 1>
    )) {isSetParameters=true, module=SetVariables_#9}
  ]
  particleSpawnScript [
   &InitializeParticle(
    Color=<rgba 6 0 0 1.5>
    Mesh Scale Mode=::NewEnumerator3
    Mesh Uniform Scale=0.5
    Sprite Size Mode=::NewEnumerator3
    Uniform Sprite Size=240
    UsePositionOffset=false
    Write Material Random=true
   )
   &InitialMeshOrientation(Mesh Axis To Orient=<0 1 0>, Mesh Orientation Mode=::NewEnumerator2, Use Rotation Vector=false)
   &CameraOffset(Camera Offset Amount=10)
  ]
  particleUpdateScript [
   &ParticleState(Kill Particles When Lifetime Has Elapsed=false, Let Infinitely Lived Particles Die When Emitter Deactivates=true)
   &ScaleMeshSize(Scale Factor=&VectorFromFloat(Value=&Multiply_Float(A=&Lerp_Float(A=0.9, B=1.1, Alpha=$Emitter.FlickerAnimAlpha), B=$System.LaserFadeOut)))
   &ScaleSpriteSize(
    Scale Sprite Size Mode=::NewEnumerator0
    Uniform Scale Factor=&Multiply_Float(A=&Lerp_Float(A=0.7, B=1.1, Alpha=$Emitter.FlickerAnimAlpha), B=$System.LaserFadeOut)
   )
   /SetVariables(
    Particles.ColorGlow:LinearColor=&MultiplyLinearColor(
     Linear Color=<rgba 1 0 0.0167 0.3>
     Scale Factor=&MakeVector4FromVectorAndFloat("Vector (RGB)"=<1 1 1>, "Float (Alpha)"=$System.LaserFadeOut)
    )
    Particles.SpriteSizeGlow:Vector2f=&Multiply_Vector2D_ByFloat(Float=1.3, Vector2D=$Particles.SpriteSize)
   ) {isSetParameters=true, module=SetVariables_#10}
   /SetVariables(
    Particles.ColorCore:LinearColor=&MultiplyLinearColor(
     Linear Color=<rgba 0.6 0 0 0.5>
     Scale Factor=&MakeVector4FromVectorAndFloat("Vector (RGB)"=<1 1 1>, "Float (Alpha)"=$System.LaserFadeOut)
    )
    Particles.SpriteSizeCore:Vector2f=&Multiply_Vector2D_ByFloat(Float=0.5, Vector2D=$Particles.SpriteSize)
   ) {isSetParameters=true, module=SetVariables_#11}
   /SetVariables(
    Particles.ColorCoreSmall:LinearColor=&MultiplyLinearColor(
     Linear Color=<rgba 3 0.6 0.6 1>
     Scale Factor=&MakeVector4FromVectorAndFloat("Vector (RGB)"=<1 1 1>, "Float (Alpha)"=$System.LaserFadeOut)
    )
    Particles.SpriteSizeCoreSmall:Vector2f=&Multiply_Vector2D_ByFloat(Float=0.8, Vector2D=$Particles.SpriteSize)
   ) {isSetParameters=true, module=SetVariables_#12}
   /SetVariables(
    Particles.ColorDark:LinearColor=&MultiplyLinearColor(
     Linear Color=<rgba 0.2 0 0 1.66>
     Scale Factor=&MakeVector4FromVectorAndFloat("Vector (RGB)"=<1 1 1>, "Float (Alpha)"=$System.LaserFadeOut)
    )
    Particles.SpriteSizeDark:Vector2f=&Multiply_Vector2D_ByFloat(Float=1, Vector2D=$Particles.SpriteSize)
   ) {isSetParameters=true}
   /SetVariables(Particles.SpriteSizeFlare:Vector2f=&Multiply_Vector2D_ByFloat(Float=0.5, Vector2D=$Particles.SpriteSize)) {enabled=false, isSetParameters=true, module=SetVariables_#13}
   /SetVariables(Particles.ScaleGlow:Vector3f=&Multiply_VectorByFloat(Vector=$Particles.Scale, Float=1.4)) {enabled=false, isSetParameters=true, module=SetVariables_#14}
   &Light_Attributes(
    Color=1
    Falloff Exponent=2
    Light Exponent=6
    Light Radius=230
    Radius=2
    Scale Alpha:NiagaraFloat=&Multiply_Float(A=6, B=$System.LaserFadeOut)
    Scale RGB:Vector3f=<1 0.8 0>
    ScaleA=true
    ScaleRGB=true
   )
  ]
  rendererClasses [&NiagaraSpriteRendererProperties, &NiagaraLightRendererProperties]
  renderers [
   {rendererClass=&NiagaraSpriteRendererProperties}
   {rendererClass=&NiagaraSpriteRendererProperties, rendererIndex=1}
   {rendererClass=&NiagaraSpriteRendererProperties, rendererIndex=2}
   {rendererClass=&NiagaraSpriteRendererProperties, rendererIndex=3}
   {rendererClass=&NiagaraLightRendererProperties, rendererIndex=4}
   {rendererClass=&NiagaraSpriteRendererProperties, rendererIndex=5}
  ]
 }
 MuzzleFlashes {
  bEnabled=false
  emitterUpdateScript [
   &EmitterState(
    Life Cycle Mode=::NewEnumerator1
    Loop Behavior=::NewEnumerator1
    Loop Delay=$System.BeamDelaySpawn
    Loop Duration Mode=::NewEnumerator1
    UseLoopDelay=true
   )
   &SpawnRate(Spawn Probability=$System.LaserFadeOut, SpawnRate=50, Use Spawn Probability=true)
  ]
  particleSpawnScript [
   &InitializeParticle(
    Color=<rgba 2 2 2 0.5>
    Lifetime=&RandomRangeFloat(Minimum=0.3, Maximum=0.45)
    Mesh Scale=&Multiply_VectorByFloat(Vector=<1 0.8 0.8>, Float=&RandomRangeFloat(Minimum=0.7, Maximum=1))
    Mesh Scale Mode=::NewEnumerator4
    Sprite Size=&Multiply_Vector2D_ByFloat(
     Float=&RandomRangeFloat(Minimum=0.8, Maximum=1)
     Vector2D=&RandomRangeVector2D(Minimum=<90 200>, Maximum=<120 260>)
    )
    Sprite Size Mode=::NewEnumerator4
    UsePositionOffset=false
    Write Material Random=true
   )
   &InitialMeshOrientation(
    Mesh Axis To Orient=<-1 0 0>
    Mesh Orientation Mode=::NewEnumerator2
    Orientation Vector=$System.BeamDirection
    Rotation=&RandomRangeVector(Minimum=<-1 0 0>, Maximum=<1 0 0>)
    Use Rotation Vector=true
   )
   &SubUVAnimation(
    "Sprite Renderer "=DataInterface() {propertyValues {SpriteRenderer=&self:MuzzleFlashes_0.NiagaraSpriteRendererProperties_1}}
    SubUV Animation Mode:ENiagara_SubUVLookupModeV2=::NewEnumerator2
    UseEndFrame=false
    UseStartFrame=false
   )
   &ShapeLocation(
    MaskSpawnGroup=false
    Override Seed=false
    Ring Radius=1
    Rotation Axis=$System.BeamDirection
    Rotation Coordinate Space=::Simulation
    Rotation Mode=::NewEnumerator1
    Shape Primitive=::NewEnumerator4
    UseOwnerScale=true
   )
   &SpriteFacingAndAlignment(
    Sprite Alignment=&Lerp_Vector(A=$System.BeamDirection, B=$Particles.ShapeLocation.ShapeTangent, Alpha=0.7)
    WriteAlignment=true
    WriteFacing=true
   )
   &AddVelocity(
    Module.Use Falloff Distance=false
    Origin Offset=&Multiply_VectorByFloat(Vector=$System.BeamDirection, Float=-3)
    Override Seed=false
    Velocity Mode=::NewEnumerator1
    Velocity Speed=&RandomRangeFloat(Minimum=30, Maximum=50)
   )
  ]
  particleUpdateScript [
   &ParticleState()
   &ScaleSpriteSize(Uniform Curve Sprite Scale=DataInterface() {propertyValues {Curve {keys [{tangentMode=RCTM_User, time=0, value=0.4}, {tangentMode=RCTM_User, time=1, value=1}]}}})
   &DynamicMaterialParameters(
    Index 0 Param 1=20
    Index 0 Param 2=1
    Index 0 Param 3=1
    Index 0 Param 4=&FloatFromCurve(FloatCurve=DataInterface() {propertyValues {Curve {keys [
         {interpMode=RCIM_Cubic, tangentMode=RCTM_User, time=0, value=0}
         {arriveTangent=2.7922, interpMode=RCIM_Cubic, leaveTangent=2.7922, tangentMode=RCTM_User, time=1, value=1}
        ]}}})
    Index 2 Param 1=0.5
    Index 2 Param 2=&FloatFromCurve(FloatCurve=DataInterface() {propertyValues {Curve {keys [
         {arriveTangent=-5.8536, interpMode=RCIM_Cubic, leaveTangent=-5.8536, tangentMode=RCTM_User, time=0, value=1}
         {arriveTangent=-0.696, interpMode=RCIM_Cubic, leaveTangent=-0.696, tangentMode=RCTM_User, time=0.2362, value=0.3031}
         {arriveTangent=-0.1613, interpMode=RCIM_Cubic, leaveTangent=-0.1613, tangentMode=RCTM_User, time=1, value=0}
        ]}}})
    Module.Param0WriteEnabled=true
    Module.Param0WriteEnabled2=true
    Module.Param1WriteEnabled=true
    Module.Param1WriteEnabled2=true
    Module.Param2WriteEnabled=true
    Module.Param2WriteEnabled2=true
    Module.Param3WriteEnabled=true
    Module.Param3WriteEnabled2=true
    Write Parameter Index 2=true
   )
   /SetVariables(
    Particles.ColorCore:LinearColor=<rgba 4 4 4 0.1>
    Particles.SpriteSizeCore:Vector2f=&Multiply_Vector2D(A=$Particles.SpriteSize, B=<0.5 0.7>)
   ) {enabled=false, isSetParameters=true, module=SetVariables_#15}
   /SetVariables(
    Particles.ColorCore:LinearColor=&ColorFromCurve(ColorCurve=DataInterface() {propertyValues {
       AlphaCurve {keys [<at 0.0008 0>,<at 0.0997 0.55 RCIM_Cubic>,<at 1 0 RCIM_Cubic>]}
       BlueCurve {keys [<at 0.0566 1 RCIM_Cubic>,<at 0.3246 0.1906>,<at 0.7329 0.0083 RCIM_Cubic>]}
       GreenCurve {keys [<at 0.0566 1 RCIM_Cubic>,<at 0.3246 0.1887>,<at 0.7329 0 RCIM_Cubic>]}
       RedCurve {keys [{arriveTangent=3.1293, interpMode=RCIM_Cubic, time=0.0566, value=5}, <at 0.3246 2>, <at 0.7329 0.5 RCIM_Cubic>]}
      }})
    Particles.SpriteSizeCore:Vector2f=&Multiply_Vector2D(A=$Particles.SpriteSize, B=<0.66 0.8>)
   ) {isSetParameters=true}
   &SolveForcesAndVelocity(Clamp Velocity=false, Limit Acceleration=false)
   &ScaleColor(
    Scale RGB=&MakeVectorFromLinearColorRGB(Color=&ColorFromCurve(ColorCurve=DataInterface() {propertyValues {
        AlphaCurve {keys [<at 0 1 RCIM_Cubic>]}
        BlueCurve {keys [<at 0.0008 0.0527 RCIM_Cubic>,<at 0.3072 0 RCIM_Cubic>]}
        GreenCurve {keys [<at 0.0008 0.02 RCIM_Cubic>,<at 0.3072 0 RCIM_Cubic>]}
        RedCurve {keys [<at 0.0008 1 RCIM_Cubic>,<at 0.3072 1 RCIM_Cubic>]}
       }}))
    ScaleA=false
    ScaleRGB=true
   )
  ]
  rendererClasses [&NiagaraSpriteRendererProperties]
  renderers [
   {rendererClass=&NiagaraSpriteRendererProperties}
   {rendererClass=&NiagaraSpriteRendererProperties, rendererIndex=1}
  ]
 }
 RocksBurst {
  bEnabled=false
  emitterUpdateScript [
   &EmitterState()
   &SpawnRate(SpawnRate=20, Use Spawn Probability=false)
   &SpawnBurst_Instantaneous(
    Spawn Count=&RandomRangeInt(Minimum=5, Maximum=10)
    Use Spawn Probability=false
    UseLoopCountLimit=false
   ) {enabled=false}
   /SetVariables() {isSetParameters=true}
  ]
  particleSpawnScript [
   &InitializeParticle(
    Lifetime=&Lerp_Float(
     A=0.4
     B=0.7
     Alpha=&FloatFromCurve(
      FloatCurve=DataInterface() {propertyValues {Curve {keys [
          {interpMode=RCIM_Cubic, tangentMode=RCTM_User, time=0, value=0}
          {arriveTangent=2.8717, interpMode=RCIM_Cubic, leaveTangent=2.8717, tangentMode=RCTM_User, time=1, value=1}
         ]}}}
      CurveIndex=&RandomRangeFloat(Minimum=0, Maximum=1)
     )
    )
    Mesh Scale Mode=::NewEnumerator3
    Mesh Uniform Scale=&Multiply_Float(
     A=&Lerp_Float(
      A=0.05
      B=0.1
      Alpha=&FloatFromCurve(
       FloatCurve=DataInterface() {propertyValues {Curve {keys [
           {interpMode=RCIM_Cubic, tangentMode=RCTM_User, time=0, value=0}
           {arriveTangent=0.6494, interpMode=RCIM_Cubic, leaveTangent=0.6494, tangentMode=RCTM_User, time=0.8, value=0.2}
           {arriveTangent=7.6187, interpMode=RCIM_Cubic, leaveTangent=7.6187, tangentMode=RCTM_User, time=1, value=1}
          ]}}}
       CurveIndex=&RandomRangeFloat(Minimum=0, Maximum=1)
      )
     )
     B=2.5
    )
    Position Offset=&Multiply_VectorByFloat(Vector=$System.BeamDirection, Float=$System.BeamImpactDistance)
    UsePositionOffset=true
    Write Material Random=true
   )
   &ShapeLocation(Handle Radius=0, Large Radius=30, MaskSpawnGroup=false, Override Seed=false, Shape Primitive=::NewEnumerator3, UseOwnerScale=true)
   &InitialMeshOrientation(Use Rotation Vector=false)
   &self:DivideParticles(HowManyRenderers=3) {module=DivideParticles}
   &AddRotationalVelocity(Rotation Rate=&Multiply_VectorByFloat(
     Vector=&RandomVector(Vector Scale=2)
     Float=&RandomRangeFloat(Minimum=0.5, Maximum=1)
    ))
   &DynamicMaterialParameters(
    Index 0 Param 2=1
    Module.Param0WriteEnabled=false
    Module.Param1WriteEnabled=true
    Module.Param2WriteEnabled=false
    Module.Param3WriteEnabled=false
   )
   &AddVelocity(
    Cone Angle=35
    Cone Axis=<0 0 1>
    Module.UseConeFalloff=false
    Module.UseVelDistribution=false
    Override Seed=false
    Velocity Mode=::NewEnumerator2
    Velocity Speed=&FloatFromCurve(
     FloatCurve=DataInterface() {propertyValues {Curve {keys [
         {interpMode=RCIM_Cubic, tangentMode=RCTM_User, time=0, value=1}
         {arriveTangent=-1.3709, interpMode=RCIM_Cubic, leaveTangent=-1.3709, tangentMode=RCTM_User, time=1, value=0}
        ]}}}
     Scale Curve=&RandomRangeFloat(Minimum=300, Maximum=666)
    )
   )
   &ApplyInitialForces()
  ]
  particleUpdateScript [
   &ParticleState()
   &GravityForce(Gravity=&RandomRangeVector(Minimum=<0 0 -750>, Maximum=<0 0 -1500>))
   &Collision(
    Age Colliding Particles=false
    Module.Control Rest State Via Surface Normal=false
    Module.EnableMaxCollisionCount=false
    Module.Randomize Collision Normal Vector=false
    Restitution=0.1
   ) {enabled=false}
   &Drag(Drag=2, Use Linear Drag=true, Use Rotational Drag=false)
   &InheritSourceMovement(
    Applied Position Scale=&VectorFromFloat(Value=&FloatFromCurve(FloatCurve=DataInterface() {propertyValues {Curve {keys [<at 0 1>,<at 0.5 0>]}}}))
    ApplyForce=false
    ApplyPosition=true
    ApplyPositionFalloff=false
   ) {enabled=false}
   &SolveForcesAndVelocity(Clamp Velocity=false, Limit Acceleration=false)
   &ScaleMeshSize(Scale Factor=&VectorFromFloat(Value=&FloatFromCurve(FloatCurve=DataInterface() {propertyValues {Curve {keys [{arriveTangent=-0.0035, interpMode=RCIM_Cubic, leaveTangent=-0.0035, tangentMode=RCTM_User, time=0.4, value=1}, <at 1 0>]}}})))
   &SolveRotationalForcesAndVelocity()
  ]
  rendererClasses [&NiagaraMeshRendererProperties]
  renderers [
   {rendererClass=&NiagaraMeshRendererProperties, rendererIndex=0}
   {rendererClass=&NiagaraMeshRendererProperties, rendererIndex=1}
   {rendererClass=&NiagaraMeshRendererProperties, rendererIndex=2}
  ]
 }
 ScorchTrail {
  bEnabled=false
  emitterUpdateScript [
   &EmitterState()
   &SpawnPerUnit(Module.Use Max Movement Threshold=true, Module.Use Movement Tolerance=false, Use Spawn Probability=false)
   &SpawnRate(SpawnRate=10, Use Spawn Probability=false) {enabled=false}
  ]
  particleSpawnScript [
   &InitializeParticle(
    Color=<rgba 1 1 1 0.7>
    Position Offset=&RotateVectorByQuaternion(Vector To Rotate=Linked() {linkedVariable {name=System.ImpactLocation, type=NiagaraPosition}}, Quaternion=$Engine.Owner.Rotation)
    Ribbon Facing Mode=::NewEnumerator1
    Ribbon Width=200
    Ribbon Width Mode=::NewEnumerator1
    UsePositionOffset=true
    Write Material Random=true
   )
   &CameraOffset(Camera Offset Amount=20) {enabled=false}
  ]
  particleUpdateScript [
   &ParticleState()
   &DynamicMaterialParameters(
    Index 0 Param 3=10
    Module.Index0Float0Write=false
    Module.Index0Float1Write=false
    Module.Index0Float2Write=true
    Module.Index0Float3Write=false
   )
   &ScaleColor(
    Scale Alpha=&Multiply_Float(
     A=&FloatFromCurve(FloatCurve=DataInterface() {propertyValues {Curve {keys [
          {arriveTangent=0.0301, interpMode=RCIM_Cubic, leaveTangent=0.0301, tangentMode=RCTM_User, time=0, value=1}
          {arriveTangent=-2.5567, interpMode=RCIM_Cubic, leaveTangent=-2.5567, tangentMode=RCTM_User, time=1, value=0}
         ]}}})
     B=&FloatFromCurve(
      FloatCurve=DataInterface() {propertyValues {Curve {keys [{interpMode=RCIM_Cubic, tangentMode=RCTM_User, time=2, value=1}, {interpMode=RCIM_Cubic, tangentMode=RCTM_User, time=3, value=0}]}}}
      CurveIndex=$Engine.System.Age
     )
    )
    Scale RGB=&VectorFromFloat(Value=0.15)
    ScaleA=true
    ScaleRGB=true
   )
   /SetVariables(
    Particles.ColorCore:LinearColor=&MultiplyLinearColor(
     Linear Color=$Particles.Color
     Scale Factor=&MakeVector4FromVectorAndFloat(
      "Vector (RGB)"=<200 20 20>
      "Float (Alpha)"=&FloatFromCurve(FloatCurve=DataInterface() {propertyValues {Curve {keys [
           {arriveTangent=-2.5221, interpMode=RCIM_Cubic, leaveTangent=-2.5221, tangentMode=RCTM_User, time=0, value=1}
           {interpMode=RCIM_Cubic, tangentMode=RCTM_User, time=0.7, value=0}
          ]}}})
     )
    )
    Particles.WidthCore:NiagaraFloat=&Multiply_Float(A=$Particles.RibbonWidth, B=0.2)
   ) {isSetParameters=true}
  ]
  rendererClasses [&NiagaraRibbonRendererProperties]
  renderers [
   {rendererClass=&NiagaraRibbonRendererProperties}
   {rendererClass=&NiagaraRibbonRendererProperties, rendererIndex=1}
  ]
 }
 Shockwave {
  bEnabled=false
  emitterUpdateScript [
   &EmitterState(
    Life Cycle Mode=::NewEnumerator1
    Loop Behavior=::NewEnumerator1
    Loop Delay=$System.BeamDelaySpawn
    Loop Duration Mode=::NewEnumerator1
    UseLoopDelay=true
   )
   &SpawnRate(Spawn Probability=$System.LaserFadeOut, SpawnRate=7, Use Spawn Probability=true)
  ]
  particleSpawnScript [
   &InitializeParticle(
    Color=<rgba 1 1 1 5>
    Lifetime=0.3
    Mesh Scale Mode=::NewEnumerator1
    Mesh Uniform Scale Max=0.65
    Mesh Uniform Scale Min=0.55
    UsePositionOffset=false
    Write Material Random=true
   )
   &InitialMeshOrientation(
    Mesh Axis To Orient=<0 0 1>
    Mesh Orientation Mode=::NewEnumerator2
    Rotation=&MakeVector(X=0, Y=0, Z=&RandomRangeFloat(Minimum=0, Maximum=1))
    Use Rotation Vector=true
   )
   &CameraOffset(Camera Offset Amount=30)
  ]
  particleUpdateScript [
   &ParticleState()
   &DynamicMaterialParameters(
    Index 0 Param 1=&FloatFromCurve(FloatCurve=DataInterface() {propertyValues {Curve {keys [{arriveTangent=-2.1015, interpMode=RCIM_Cubic, leaveTangent=-2.1015, tangentMode=RCTM_User, time=0, value=0.5}, <at 1 -0.15>]}}})
    Index 0 Param 2=&FloatFromCurve(FloatCurve=DataInterface() {propertyValues {Curve {keys [{arriveTangent=-0.3464, interpMode=RCIM_Cubic, leaveTangent=-0.3464, tangentMode=RCTM_User, time=0, value=1}, <at 1 0.9>]}}})
    Index 0 Param 3=&FloatFromCurve(FloatCurve=DataInterface() {propertyValues {Curve {keys [<at 0 0>,<at 0.15 1>,<at 1 1.2>]}}})
    Module.Param0WriteEnabled=true
    Module.Param1WriteEnabled=true
    Module.Param2WriteEnabled=true
    Module.Param3WriteEnabled=false
   ) {module=DynamicMaterialParameters001}
   &ScaleColor(
    Linear Color Curve=DataInterface() {propertyValues {
      AlphaCurve {keys [<at 0 0.5>,<at 0.967 0>]}
      BlueCurve {keys [<at 0 0.086>,<at 0.6282 0.0167>]}
      GreenCurve {keys [<at 0 0.02>,<at 0.6282 0>]}
      RedCurve {keys [<at 0 2>,<at 0.6282 1>]}
     }}
    Scale Mode=::NewEnumerator2
   )
   &ScaleMeshSize(Scale Factor=&VectorFromFloat(Value=&FloatFromCurve(
      FloatCurve=DataInterface() {propertyValues {Curve {keys [
          {arriveTangent=1.0704, interpMode=RCIM_Cubic, leaveTangent=1.0704, tangentMode=RCTM_User, time=0, value=0.5}
          {interpMode=RCIM_Cubic, tangentMode=RCTM_User, time=1, value=1}
         ]}}}
      Scale Curve=2
     )))
  ]
  rendererClasses [&NiagaraMeshRendererProperties]
  renderers [{rendererClass=&NiagaraMeshRendererProperties, rendererIndex=0}]
 }
 ShockwaveLW {rendererClasses [], renderers []}
 Sparks {bEnabled=false, rendererClasses [], renderers []}
 StartGlow_White {rendererClasses [], renderers []}
 StartGlow_White001 {rendererClasses [], renderers []}
}
summary {
 systemName=NS_SouthBulk_Boss_EnergyRay
 emitters [
  {emitterName=MuzzleCore, rendererClasses [&NiagaraSpriteRendererProperties, &NiagaraLightRendererProperties]}
  {bEnabled=false, emitterName=MuzzleFlashes, rendererClasses [&NiagaraSpriteRendererProperties]}
  {emitterName=Minimal005, rendererClasses [&NiagaraSpriteRendererProperties]}
  {emitterName=StartGlow_White, rendererClasses []}
  {emitterName=StartGlow_White001, rendererClasses []}
  {bEnabled=false, emitterName=ImpactFlashes, rendererClasses [&NiagaraSpriteRendererProperties]}
  {
   bEnabled=false
   emitterName=ImpactCore
   rendererClasses [&NiagaraSpriteRendererProperties, &NiagaraLightRendererProperties]
  }
  {emitterName=Beam, rendererClasses [&NiagaraRibbonRendererProperties, &NiagaraSpriteRendererProperties]}
  {bEnabled=false, emitterName=RocksBurst, rendererClasses [&NiagaraMeshRendererProperties]}
  {bEnabled=false, emitterName=ImpactRocks, rendererClasses [&NiagaraMeshRendererProperties]}
  {bEnabled=false, emitterName=Sparks, rendererClasses []}
  {bEnabled=false, emitterName=ScorchTrail, rendererClasses [&NiagaraRibbonRendererProperties]}
  {bEnabled=false, emitterName=Shockwave, rendererClasses [&NiagaraMeshRendererProperties]}
  {bEnabled=false, emitterName=EnergyThorn, rendererClasses [&NiagaraMeshRendererProperties]}
  {emitterName=ShockwaveLW, rendererClasses []}
  {emitterName=EnergyThornLW, rendererClasses []}
 ]
 userVariables [User.EnergyRayDirection {type=Vector3f, description="", defaultValue {struct=Vector3f, value=<1 0 -0.35>}}]
}
