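# Niagara system for the impact point of the SouthBulk boss's energy ray, with emitters positioned at System.ImpactLocation.
# A flickering sprite-and-light core combines with thorn and shockwave meshes, GPU sparks, falling rocks and a distance-spawned ribbon scorch trail.
# System.BeamDelaySpawn delays most emitters, while rocks ramp on system age.
# System.LaserFadeOut tapers spawn probability and core size, alpha and light; ImpactFlashes, ImpactCoreOld and Sparks are disabled.
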NiagaraSystem /SouthBulkAI/Assets/VFX/Laser/NS_SouthBulk_Boss_EnergyRay_Impact
dependencies {
 usedDataInterfaces [
  &NiagaraDataInterfaceSpriteRendererInfo
  &NiagaraDataInterfaceCurve
  &NiagaraDataInterfaceColorCurve
 ]
 usedDynamicInputs [
  &MultiplyQuaternion
  &MakeVector4_FromLinearColor
  &LerpLinearColors
 ]
 usedModules [
  &self:SystemSpawnScript.SystemScriptSource.SystemScriptGraph.NiagaraNodeAssignment_6.SetVariables_#1
  &SystemState
  &self:SystemSpawnScript.SystemScriptSource.SystemScriptGraph.NiagaraNodeAssignment_5.SetVariables_#2
  &self:SystemSpawnScript.SystemScriptSource.SystemScriptGraph.NiagaraNodeAssignment_0.SetVariables_#3
  &self:SystemSpawnScript.SystemScriptSource.SystemScriptGraph.NiagaraNodeAssignment_1.SetVariables_#4
  &self:SystemSpawnScript.SystemScriptSource.SystemScriptGraph.NiagaraNodeAssignment_2.SetVariables_#5
  &self:SystemSpawnScript.SystemScriptSource.SystemScriptGraph.NiagaraNodeAssignment_3.SetVariables_#6
  &self:SystemSpawnScript.SystemScriptSource.SystemScriptGraph.NiagaraNodeAssignment_4.SetVariables_#7
 ]
}
emitters {
 EnergyThorn {
  emitterUpdateScript [
   &EmitterState(
    Life Cycle Mode=::NewEnumerator1
    Loop Behavior=::NewEnumerator1
    Loop Delay=$System.BeamDelaySpawn
    Loop Duration Mode=::NewEnumerator1
    UseLoopDelay=true
   )
   &SpawnBurst_Instantaneous(Spawn Count=3, Use Spawn Probability=false, UseLoopCountLimit=false) {enabled=false}
   &SpawnRate(Spawn Probability=$System.LaserFadeOut, SpawnRate=30, Use Spawn Probability=true)
  ]
  particleSpawnScript [
   &InitializeParticle(
    Color Maximum=<rgba 2 0.1 0.1 1>
    Color Minimum=<rgba 1.5 0.03 0.079 1>
    Color Mode=::NewEnumerator2
    Lifetime Max=0.2
    Lifetime Min=0.15
    Lifetime Mode=::NewEnumerator1
    Mesh Scale Max=<3.5 3.5 3.5>
    Mesh Scale Min=<2.5 2.5 2.5>
    Mesh Scale Mode=::NewEnumerator2
    Position Offset:NiagaraPosition=$System.ImpactLocation
    UsePositionOffset=true
    Write Material Random=true
   )
   &AddVelocity(
    Module.Use Falloff Distance=false
    Origin Offset:NiagaraPosition=$System.ImpactLocation
    Override Seed=false
    Velocity Mode=::NewEnumerator1
    Velocity Speed=1
   )
   &InitialMeshOrientation(
    Mesh Orientation Mode=::NewEnumerator2
    Rotation=&RandomRangeVector(Minimum=<0 0 0>, Maximum=<1 0 0>)
    Use Rotation Vector=true
   )
  ]
  particleUpdateScript [
   &SolveForcesAndVelocity(Clamp Velocity=false, Limit Acceleration=false)
   &ParticleState()
   &ScaleMeshSize(Scale Factor=&Multiply_Vector(
     A=&MakeVector(
      X=&FloatFromCurve(FloatCurve=DataInterface() {propertyValues {Curve {keys [<at 0 0.2>,<at 1 1>]}}})
      Y=&FloatFromCurve(FloatCurve=DataInterface() {propertyValues {Curve {keys [{interpMode=RCIM_Cubic, tangentMode=RCTM_User, time=0, value=0.5}, {interpMode=RCIM_Cubic, tangentMode=RCTM_User, time=1, value=2}]}}})
      Z=&FloatFromCurve(FloatCurve=DataInterface() {propertyValues {Curve {keys [{interpMode=RCIM_Cubic, tangentMode=RCTM_User, time=0, value=0.5}, {interpMode=RCIM_Cubic, tangentMode=RCTM_User, time=1, value=2}]}}})
     )
     B=<1 1 1>
    ))
   &ScaleColor(
    Scale Alpha=&FloatFromCurve(
     FloatCurve=DataInterface() {propertyValues {Curve {keys [
         {arriveTangent=2.8247, interpMode=RCIM_Cubic, leaveTangent=2.8247, tangentMode=RCTM_User, time=0, value=0}
         {arriveTangent=0.0384, time=0.1, value=1}
         {arriveTangent=-1.1111, interpMode=RCIM_Cubic, leaveTangent=-1.1111, time=0.5, value=1}
         {arriveTangent=0.0424, interpMode=RCIM_Cubic, leaveTangent=0.0424, tangentMode=RCTM_User, time=1, value=0}
        ]}}}
     Scale Curve=&Multiply_Float(A=0.8, B=$System.LaserFadeOut)
    )
    Scale RGB=&MakeVectorFromLinearColorRGB(Color=&ColorFromCurve(ColorCurve=DataInterface() {propertyValues {
        AlphaCurve {keys []}
        BlueCurve {keys [{arriveTangent=-6.2021, time=0.1407, value=2}, {arriveTangent=-1.9715, time=0.4613, value=1}, {leaveTangent=-0.0154, time=0.9275, value=0.1}]}
        GreenCurve {keys [
          {arriveTangent=-6.099, time=0.1407, value=2}
          {arriveTangent=0.1945, time=0.4613, value=1}
          {arriveTangent=-0.0129, leaveTangent=-0.0129, time=0.9275, value=0.1}
         ]}
        RedCurve {keys [
          {arriveTangent=-31.4411, time=0.1407, value=2}
          {arriveTangent=-21.5603, time=0.4613, value=1}
          {leaveTangent=-0.0154, time=0.9275, value=0.1}
         ]}
       }}))
    ScaleA=true
    ScaleRGB=true
   )
  ]
  rendererClasses [&NiagaraMeshRendererProperties]
  renderers [{rendererClass=&NiagaraMeshRendererProperties, rendererIndex=0}]
 }
 ImpactCore {
  emitterUpdateScript [
   &EmitterState(
    Life Cycle Mode=::NewEnumerator1
    Loop Behavior=::NewEnumerator1
    Loop Delay=&Add_Float(A=$System.BeamDelaySpawn, B=0.15)
    Loop Duration Mode=::NewEnumerator1
    UseLoopDelay=true
   )
   &SpawnBurst_Instantaneous(Use Spawn Probability=false, UseLoopCountLimit=false)
   /SetVariables(Emitter.FlickerAnimAlpha:NiagaraFloat=&Waveform(
     "Waveform [1]"=::NewEnumerator6
     "[1] Input Position"=&RandomRangeFloat(Minimum=0, Maximum=1)
     "[1] Frequency"=7
     "[1] Amplitude Min/Max"=<0 1>
    )) {isSetParameters=true, module=SetVariables_#8}
  ]
  particleSpawnScript [
   &InitializeParticle(
    Color=<rgba 6 0 0 1.5>
    Mesh Scale Mode=::NewEnumerator3
    Mesh Uniform Scale=0.5
    Position Offset:NiagaraPosition=$System.ImpactLocation
    Sprite Size Mode=::NewEnumerator3
    Uniform Sprite Size=240
    UsePositionOffset=true
    Write Material Random=true
   )
   &InitialMeshOrientation(Mesh Axis To Orient=<0 1 0>, Mesh Orientation Mode=::NewEnumerator2, Use Rotation Vector=false)
   &CameraOffset(Camera Offset Amount=30)
  ]
  particleUpdateScript [
   &ParticleState(Kill Particles When Lifetime Has Elapsed=false, Let Infinitely Lived Particles Die When Emitter Deactivates=true)
   &ScaleMeshSize(Scale Factor=&VectorFromFloat(Value=&Multiply_Float(A=&Lerp_Float(A=0.9, B=1.1, Alpha=$Emitter.FlickerAnimAlpha), B=$System.LaserFadeOut)))
   &ScaleSpriteSize(
    Scale Sprite Size Mode=::NewEnumerator0
    Uniform Scale Factor=&Multiply_Float(A=&Lerp_Float(A=0.7, B=1.1, Alpha=$Emitter.FlickerAnimAlpha), B=$System.LaserFadeOut)
   )
   /SetVariables(
    Particles.ColorGlow:LinearColor=&MultiplyLinearColor(
     Linear Color=<rgba 1 0 0.05 0.3>
     Scale Factor=&MakeVector4FromVectorAndFloat("Vector (RGB)"=<1 1 1>, "Float (Alpha)"=$System.LaserFadeOut)
    )
    Particles.SpriteSizeGlow:Vector2f=&Multiply_Vector2D_ByFloat(Float=1.3, Vector2D=$Particles.SpriteSize)
   ) {isSetParameters=true, module=SetVariables_#9}
   /SetVariables(
    Particles.ColorCore:LinearColor=&MultiplyLinearColor(
     Linear Color=<rgba 0 0 0 1>
     Scale Factor=&MakeVector4FromVectorAndFloat("Vector (RGB)"=<1 1 1>, "Float (Alpha)"=$System.LaserFadeOut)
    )
    Particles.SpriteSizeCore:Vector2f=&Multiply_Vector2D_ByFloat(Float=0.4, Vector2D=$Particles.SpriteSize)
   ) {isSetParameters=true, module=SetVariables_#10}
   /SetVariables(
    Particles.ColorCoreSmall:LinearColor=&MultiplyLinearColor(
     Linear Color=<rgba 3 0.24 0.286 1>
     Scale Factor=&MakeVector4FromVectorAndFloat("Vector (RGB)"=<1 1 1>, "Float (Alpha)"=$System.LaserFadeOut)
    )
    Particles.SpriteSizeCoreSmall:Vector2f=&Multiply_Vector2D_ByFloat(Float=1, Vector2D=$Particles.SpriteSize)
   ) {isSetParameters=true, module=SetVariables_#11}
   /SetVariables(
    Particles.ColorDark:LinearColor=&MultiplyLinearColor(
     Linear Color=<rgba 0 0 0 2>
     Scale Factor=&MakeVector4FromVectorAndFloat("Vector (RGB)"=<1 1 1>, "Float (Alpha)"=$System.LaserFadeOut)
    )
    Particles.SpriteSizeDark:Vector2f=&Multiply_Vector2D_ByFloat(Float=0.6, Vector2D=$Particles.SpriteSize)
   ) {isSetParameters=true}
   /SetVariables(Particles.SpriteSizeFlare:Vector2f=&Multiply_Vector2D_ByFloat(Float=0.5, Vector2D=$Particles.SpriteSize)) {enabled=false, isSetParameters=true, module=SetVariables_#12}
   /SetVariables(Particles.ScaleGlow:Vector3f=&Multiply_VectorByFloat(Vector=$Particles.Scale, Float=1.4)) {enabled=false, isSetParameters=true, module=SetVariables_#13}
   &Light_Attributes(
    Color=1
    Falloff Exponent=2
    Light Exponent=6
    Light Radius=230
    Radius=2
    Scale Alpha:NiagaraFloat=&Multiply_Float(A=6, B=$System.LaserFadeOut)
    Scale RGB:Vector3f=<1 0.8 0>
    ScaleA=true
    ScaleRGB=true
   )
  ]
  rendererClasses [&NiagaraSpriteRendererProperties, &NiagaraLightRendererProperties]
  renderers [
   {rendererClass=&NiagaraSpriteRendererProperties}
   {rendererClass=&NiagaraSpriteRendererProperties, rendererIndex=1}
   {rendererClass=&NiagaraSpriteRendererProperties, rendererIndex=2}
   {rendererClass=&NiagaraSpriteRendererProperties, rendererIndex=3}
   {rendererClass=&NiagaraLightRendererProperties, rendererIndex=4}
   {rendererClass=&NiagaraSpriteRendererProperties, rendererIndex=5}
  ]
 }
 ImpactCoreOld {
  bEnabled=false
  emitterUpdateScript [
   &EmitterState(
    Life Cycle Mode=::NewEnumerator1
    Loop Behavior=::NewEnumerator1
    Loop Delay=&Add_Float(A=$System.BeamDelaySpawn, B=0.15)
    Loop Duration Mode=::NewEnumerator1
    UseLoopDelay=true
   )
   &SpawnBurst_Instantaneous(Use Spawn Probability=false, UseLoopCountLimit=false)
   /SetVariables(Emitter.FlickerAnimAlpha:NiagaraFloat=&Waveform(
     "Waveform [1]"=::NewEnumerator6
     "[1] Input Position"=&RandomRangeFloat(Minimum=0, Maximum=1)
     "[1] Frequency"=7
     "[1] Amplitude Min/Max"=<0 1>
    )) {isSetParameters=true, module=SetVariables_#8}
  ]
  particleSpawnScript [
   &InitializeParticle(
    Color=<rgba 6 0 0 1.5>
    Mesh Scale Mode=::NewEnumerator3
    Mesh Uniform Scale=0.4
    Position Offset:NiagaraPosition=$System.ImpactLocation
    Sprite Size Mode=::NewEnumerator3
    Uniform Sprite Size=400
    UsePositionOffset=true
    Write Material Random=true
   )
   &InitialMeshOrientation(Mesh Axis To Orient=<0 1 0>, Mesh Orientation Mode=::NewEnumerator2, Use Rotation Vector=false)
  ]
  particleUpdateScript [
   &ParticleState(Kill Particles When Lifetime Has Elapsed=false, Let Infinitely Lived Particles Die When Emitter Deactivates=true)
   &ScaleMeshSize(Scale Factor=&VectorFromFloat(Value=&Multiply_Float(A=&Lerp_Float(A=0.8, B=1.3, Alpha=$Emitter.FlickerAnimAlpha), B=$System.LaserFadeOut)))
   &ScaleSpriteSize(
    Scale Sprite Size Mode=::NewEnumerator0
    Uniform Scale Factor=&Multiply_Float(A=&Lerp_Float(A=0.7, B=1.1, Alpha=$Emitter.FlickerAnimAlpha), B=$System.LaserFadeOut)
   )
   /SetVariables(
    Particles.ColorGlow:LinearColor=<rgba 1 0 0 1>
    Particles.SpriteSizeGlow:Vector2f=&Multiply_Vector2D_ByFloat(Float=1, Vector2D=$Particles.SpriteSize)
   ) {isSetParameters=true, module=SetVariables_#9}
   /SetVariables(
    Particles.ColorCore:LinearColor=<rgba 1 0.02 0.0527 2>
    Particles.SpriteSizeCore:Vector2f=&Multiply_Vector2D_ByFloat(Float=0.3, Vector2D=$Particles.SpriteSize)
   ) {isSetParameters=true, module=SetVariables_#10}
   /SetVariables(
    Particles.ColorCoreSmall:LinearColor=<rgba 2 0.5067 0.4 1>
    Particles.SpriteSizeCoreSmall:Vector2f=&Multiply_Vector2D_ByFloat(Float=0.5, Vector2D=$Particles.SpriteSize)
   ) {isSetParameters=true, module=SetVariables_#11}
   /SetVariables(
    Particles.ColorFlare:LinearColor=<rgba 5 0 0 0.2>
    Particles.SpriteSizeFlare:Vector2f=&Multiply_Vector2D_ByFloat(Float=0.55, Vector2D=$Particles.SpriteSize)
   ) {isSetParameters=true, module=SetVariables_#12}
   /SetVariables(Particles.ScaleGlow:Vector3f=&Multiply_VectorByFloat(Vector=$Particles.Scale, Float=1.4)) {isSetParameters=true, module=SetVariables_#13}
   &Light_Attributes(
    Color=1
    Falloff Exponent=2
    Light Exponent=6
    Light Radius=300
    Radius=2
    Scale Alpha:NiagaraFloat=&Multiply_Float(A=&Lerp_Float(A=2, B=7, Alpha=$Emitter.FlickerAnimAlpha), B=$System.LaserFadeOut)
    ScaleA=true
    ScaleRGB=true
   )
   &CameraOffset(Camera Offset Amount=30)
  ]
  rendererClasses [&NiagaraSpriteRendererProperties, &NiagaraLightRendererProperties]
  renderers [
   {rendererClass=&NiagaraSpriteRendererProperties}
   {rendererClass=&NiagaraSpriteRendererProperties, rendererIndex=1}
   {rendererClass=&NiagaraLightRendererProperties, rendererIndex=2}
   {rendererClass=&NiagaraSpriteRendererProperties, rendererIndex=3}
   {rendererClass=&NiagaraSpriteRendererProperties, rendererIndex=4}
  ]
 }
 ImpactFlashes {
  bEnabled=false
  emitterUpdateScript [
   &EmitterState(
    Life Cycle Mode=::NewEnumerator1
    Loop Behavior=::NewEnumerator1
    Loop Delay=&Add_Float(A=$System.BeamDelaySpawn, B=0.15)
    Loop Duration Mode=::NewEnumerator1
    UseLoopDelay=true
   )
   &SpawnRate(Spawn Probability=$System.LaserFadeOut, SpawnRate=20, Use Spawn Probability=true)
  ]
  particleSpawnScript [
   &InitializeParticle(
    Color=<rgba 2 2 2 0.5>
    Lifetime=&RandomRangeFloat(Minimum=0.27, Maximum=0.4)
    Sprite Size=&Multiply_Vector2D_ByFloat(
     Float=&RandomRangeFloat(Minimum=2, Maximum=3)
     Vector2D=&RandomRangeVector2D(Minimum=<50 200>, Maximum=<80 220>)
    )
    Sprite Size Mode=::NewEnumerator4
    UsePositionOffset=false
    Write Material Random=true
   )
   &InitialMeshOrientation(
    Mesh Axis To Orient=<-1 0 0>
    Mesh Orientation Mode=::NewEnumerator2
    Orientation Vector=$System.BeamDirection
    Rotation=&RandomRangeVector(Minimum=<-1 0 0>, Maximum=<1 0 0>)
    Use Rotation Vector=true
   )
   &SubUVAnimation(
    "Sprite Renderer "=DataInterface() {propertyValues {SpriteRenderer=&self:ImpactFlashes_0.NiagaraSpriteRendererProperties_1}}
    SubUV Animation Mode:ENiagara_SubUVLookupModeV2=::NewEnumerator2
    UseEndFrame=false
    UseStartFrame=false
   )
   &ShapeLocation(
    MaskSpawnGroup=false
    Offset:NiagaraPosition=$System.ImpactLocation
    Offset Coordinate Space=::Local
    Offset Mode=::NewEnumerator0
    Override Seed=false
    Rotation Coordinate Space=::Simulation
    Sphere Radius=6
    UseOwnerScale=true
   )
   &SpriteFacingAndAlignment(
    Sprite Alignment=&Lerp_Vector(A=$System.BeamDirection, B=$Particles.ShapeLocation.ShapeTangent, Alpha=0.7)
    WriteAlignment=true
    WriteFacing=true
   ) {enabled=false}
   &AddVelocity(
    Module.Use Falloff Distance=false
    Origin Offset=&Multiply_VectorByFloat(
     Vector=&AddVectorToPosition(
      Position=Linked() {linkedVariable {name=System.BeamDirection, type=Vector3f}}
      Vector=&RandomRangeVector(Minimum=<0.02 -0.1 -0.1>, Maximum=<0.1 0.1 -0.2>)
     )
     Float=380
    )
    Override Seed=false
    Velocity Mode=::NewEnumerator1
    Velocity Speed=&RandomRangeFloat(Minimum=10, Maximum=20)
   )
   &CameraOffset(Camera Offset Amount=20)
  ]
  particleUpdateScript [
   &ParticleState()
   &ScaleSpriteSize(Uniform Curve Sprite Scale=DataInterface() {propertyValues {Curve {keys [{tangentMode=RCTM_User, time=0, value=0.4}, {tangentMode=RCTM_User, time=1, value=1}]}}})
   &DynamicMaterialParameters(
    Index 0 Param 1=20
    Index 0 Param 2=1
    Index 0 Param 3=1
    Index 0 Param 4=&FloatFromCurve(FloatCurve=DataInterface() {propertyValues {Curve {keys [
         {interpMode=RCIM_Cubic, tangentMode=RCTM_User, time=0, value=0}
         {arriveTangent=1.0951, interpMode=RCIM_Cubic, leaveTangent=1.0951, tangentMode=RCTM_User, time=1, value=1}
        ]}}})
    Index 2 Param 1=0.5
    Index 2 Param 2=&FloatFromCurve(FloatCurve=DataInterface() {propertyValues {Curve {keys [
         {arriveTangent=-3.0317, interpMode=RCIM_Cubic, leaveTangent=-3.0317, tangentMode=RCTM_User, time=0, value=1}
         {arriveTangent=-0.1613, interpMode=RCIM_Cubic, leaveTangent=-0.1613, tangentMode=RCTM_User, time=1, value=0}
        ]}}})
    Module.Param0WriteEnabled=true
    Module.Param0WriteEnabled2=true
    Module.Param1WriteEnabled=false
    Module.Param1WriteEnabled2=true
    Module.Param2WriteEnabled=false
    Module.Param2WriteEnabled2=true
    Module.Param3WriteEnabled=true
    Module.Param3WriteEnabled2=true
    Write Parameter Index 2=true
   )
   &ScaleColor(
    Scale Alpha=&FloatFromCurve(FloatCurve=DataInterface() {propertyValues {Curve {keys [
         {interpMode=RCIM_Cubic, tangentMode=RCTM_User, time=0, value=0}
         {interpMode=RCIM_Cubic, tangentMode=RCTM_User, time=0.15, value=1}
         {arriveTangent=-3.1002, interpMode=RCIM_Cubic, leaveTangent=-3.1002, tangentMode=RCTM_User, time=1, value=0}
        ]}}})
    Scale RGB=&MakeVectorFromLinearColorRGB(Color=&ColorFromCurve(ColorCurve=DataInterface() {propertyValues {
        AlphaCurve {keys [<at 0.1519 1 RCIM_Cubic>]}
        BlueCurve {keys [<at 0.0008 0.069 RCIM_Cubic>,<at 0.3233 0 RCIM_Cubic>]}
        GreenCurve {keys [<at 0.0008 0.02 RCIM_Cubic>,<at 0.3233 0 RCIM_Cubic>]}
        RedCurve {keys [<at 0.0008 1 RCIM_Cubic>,<at 0.3233 1 RCIM_Cubic>]}
       }}))
    ScaleA=true
    ScaleRGB=true
   )
   /SetVariables(
    Particles.ColorCore:LinearColor=&ColorFromCurve(ColorCurve=DataInterface() {propertyValues {
       AlphaCurve {keys [<at 0.0008 0>,<at 0.0997 0.55 RCIM_Cubic>,<at 1 0 RCIM_Cubic>]}
       BlueCurve {keys [<at 0.0566 1 RCIM_Cubic>,<at 0.3246 0.1906>,<at 0.7329 0.0083 RCIM_Cubic>]}
       GreenCurve {keys [<at 0.0566 1 RCIM_Cubic>,<at 0.3246 0.1887>,<at 0.7329 0 RCIM_Cubic>]}
       RedCurve {keys [{arriveTangent=3.1293, interpMode=RCIM_Cubic, time=0.0566, value=5}, <at 0.3246 2>, <at 0.7329 0.5 RCIM_Cubic>]}
      }})
    Particles.SpriteSizeCore:Vector2f=&Multiply_Vector2D(A=$Particles.SpriteSize, B=<0.66 0.8>)
   ) {isSetParameters=true}
   &SolveForcesAndVelocity(Clamp Velocity=false, Limit Acceleration=false)
  ]
  rendererClasses [&NiagaraSpriteRendererProperties]
  renderers [
   {rendererClass=&NiagaraSpriteRendererProperties}
   {rendererClass=&NiagaraSpriteRendererProperties, rendererIndex=1}
  ]
 }
 ImpactRocks {
  emitterUpdateScript [
   &EmitterState()
   &SpawnRate(
    Spawn Probability=$System.LaserFadeOut
    SpawnRate=&FloatFromCurve(
     FloatCurve=DataInterface() {propertyValues {Curve {keys [
         {arriveTangent=1.7658, interpMode=RCIM_Cubic, leaveTangent=1.7658, tangentMode=RCTM_User, time=0.3, value=0}
         {interpMode=RCIM_Cubic, tangentMode=RCTM_User, time=0.66, value=1}
        ]}}}
     CurveIndex=$Engine.System.Age
     Scale Curve=66
    )
    Use Spawn Probability=true
   )
  ]
  particleSpawnScript [
   &InitializeParticle(
    Color Mode=::NewEnumerator0
    Lifetime=&RandomRangeFloat(Minimum=0.35, Maximum=0.6)
    Mesh Scale=&Multiply_VectorByFloat(
     Vector=&RandomRangeVector(Minimum=<0.8 0.8 0.8>, Maximum=<1 1 1>)
     Float=&RandomRangeFloat(Minimum=0.1, Maximum=0.4)
    )
    Mesh Scale Mode=::NewEnumerator4
    UsePositionOffset=false
    Write Material Random=true
   )
   &ShapeLocation(
    MaskSpawnGroup=false
    Offset:NiagaraPosition=$System.ImpactLocation
    Offset Coordinate Space=::Local
    Offset Mode=::NewEnumerator0
    Override Seed=false
    Sphere Radius=33
    Sphere Surface Distribution=0.2738
    UseOwnerScale=true
   )
   &AddVelocity(
    Module.Use Falloff Distance=false
    Origin Offset=&AddVectorToPosition(
     Position=&Multiply_VectorByFloat(Vector=Linked() {linkedVariable {name=System.ImpactLocation, type=NiagaraPosition}}, Float=1.15)
     Vector=<0 0 0>
    )
    Override Seed=false
    Velocity Mode=::NewEnumerator1
    Velocity Speed=&RandomRangeFloat(Minimum=700, Maximum=1000)
   )
   &InitialMeshOrientation(Use Rotation Vector=false)
  ]
  particleUpdateScript [
   &ParticleState()
   &ScaleMeshSize(Scale Factor=&VectorFromFloat(Value=&FloatFromCurve(FloatCurve=DataInterface() {propertyValues {Curve {keys [
          {arriveTangent=1.9167, interpMode=RCIM_Cubic, leaveTangent=1.9167, tangentMode=RCTM_User, time=0, value=0}
          {interpMode=RCIM_Cubic, tangentMode=RCTM_User, time=0.22, value=1}
          {interpMode=RCIM_Cubic, tangentMode=RCTM_User, time=0.7, value=1}
          {arriveTangent=-4.0991, interpMode=RCIM_Cubic, leaveTangent=-4.0991, tangentMode=RCTM_User, time=1, value=0}
         ]}}})))
   &UpdateMeshOrientation(
    Rotation Rate=&RandomRangeFloat(Minimum=1, Maximum=2)
    Rotation Vector=&RandomRangeVector(Minimum=<-1 -1 -1>, Maximum=<1 1 1>)
   )
   &Drag(Drag=&RandomRangeFloat(Minimum=0.66, Maximum=1.1), Use Linear Drag=true, Use Rotational Drag=true)
   &GravityForce(Gravity=&Multiply_VectorByFloat(Vector=<0 0 -980>, Float=&RandomRangeFloat(Minimum=1, Maximum=2)))
   &InheritSourceMovement(
    Applied Position Scale=&VectorFromFloat(Value=&FloatFromCurve(FloatCurve=DataInterface() {propertyValues {Curve {keys [
          {arriveTangent=-2.7977, interpMode=RCIM_Cubic, leaveTangent=-2.7977, tangentMode=RCTM_User, time=0, value=1}
          {interpMode=RCIM_Cubic, tangentMode=RCTM_User, time=0.5, value=0}
         ]}}}))
    ApplyForce=false
    ApplyPosition=true
    ApplyPositionFalloff=false
   )
   &SolveForcesAndVelocity(Clamp Velocity=false, Limit Acceleration=false)
  ]
  rendererClasses [&NiagaraMeshRendererProperties]
  renderers [{rendererClass=&NiagaraMeshRendererProperties, rendererIndex=0}]
 }
 ScorchTrail {
  emitterUpdateScript [
   &EmitterState()
   &SpawnPerUnit(Module.Use Max Movement Threshold=true, Module.Use Movement Tolerance=false, Use Spawn Probability=false)
   &SpawnRate(SpawnRate=10, Use Spawn Probability=false) {enabled=false}
  ]
  particleSpawnScript [
   &InitializeParticle(
    Color=<rgba 1 1 1 0.7>
    Position Offset:NiagaraPosition=$System.ImpactLocation
    Ribbon Facing Mode=::NewEnumerator1
    Ribbon Width=200
    Ribbon Width Mode=::NewEnumerator1
    UsePositionOffset=true
    Write Material Random=true
   )
   &CameraOffset(Camera Offset Amount=20) {enabled=false}
  ]
  particleUpdateScript [
   &ParticleState()
   &DynamicMaterialParameters(
    Index 0 Param 3=10
    Module.Index0Float0Write=false
    Module.Index0Float1Write=false
    Module.Index0Float2Write=true
    Module.Index0Float3Write=false
   )
   &ScaleColor(
    Scale Alpha=&Multiply_Float(
     A=&FloatFromCurve(FloatCurve=DataInterface() {propertyValues {Curve {keys [
          {arriveTangent=0.0301, interpMode=RCIM_Cubic, leaveTangent=0.0301, tangentMode=RCTM_User, time=0, value=1}
          {arriveTangent=-2.5567, interpMode=RCIM_Cubic, leaveTangent=-2.5567, tangentMode=RCTM_User, time=1, value=0}
         ]}}})
     B=&FloatFromCurve(
      FloatCurve=DataInterface() {propertyValues {Curve {keys [{interpMode=RCIM_Cubic, tangentMode=RCTM_User, time=2, value=1}, {interpMode=RCIM_Cubic, tangentMode=RCTM_User, time=3, value=0}]}}}
      CurveIndex=$Engine.System.Age
     )
    )
    Scale RGB=&VectorFromFloat(Value=0.15)
    ScaleA=true
    ScaleRGB=true
   )
   /SetVariables(
    Particles.ColorCore:LinearColor=&MultiplyLinearColor(
     Linear Color=$Particles.Color
     Scale Factor=&MakeVector4FromVectorAndFloat(
      "Vector (RGB)"=<200 20 20>
      "Float (Alpha)"=&FloatFromCurve(FloatCurve=DataInterface() {propertyValues {Curve {keys [
           {arriveTangent=-2.5221, interpMode=RCIM_Cubic, leaveTangent=-2.5221, tangentMode=RCTM_User, time=0, value=1}
           {interpMode=RCIM_Cubic, tangentMode=RCTM_User, time=0.7, value=0}
          ]}}})
     )
    )
    Particles.WidthCore:NiagaraFloat=&Multiply_Float(A=$Particles.RibbonWidth, B=0.2)
   ) {isSetParameters=true}
  ]
  rendererClasses [&NiagaraRibbonRendererProperties]
  renderers [
   {rendererClass=&NiagaraRibbonRendererProperties}
   {rendererClass=&NiagaraRibbonRendererProperties, rendererIndex=1}
  ]
 }
 Shockwave {
  emitterUpdateScript [
   &EmitterState(
    Life Cycle Mode=::NewEnumerator1
    Loop Behavior=::NewEnumerator1
    Loop Delay=$System.BeamDelaySpawn
    Loop Duration Mode=::NewEnumerator1
    UseLoopDelay=true
   )
   &SpawnRate(Spawn Probability=$System.LaserFadeOut, SpawnRate=7, Use Spawn Probability=true)
  ]
  particleSpawnScript [
   &InitializeParticle(
    Color=<rgba 1 1 1 5>
    Lifetime=0.3
    Mesh Scale Mode=::NewEnumerator1
    Mesh Uniform Scale Max=0.65
    Mesh Uniform Scale Min=0.55
    Position Offset:NiagaraPosition=$System.ImpactLocation
    UsePositionOffset=true
    Write Material Random=true
   )
   &InitialMeshOrientation(
    Mesh Axis To Orient=<0 0 1>
    Mesh Orientation Mode=::NewEnumerator2
    Rotation=&MakeVector(X=0, Y=0, Z=&RandomRangeFloat(Minimum=0, Maximum=1))
    Use Rotation Vector=true
   )
   &CameraOffset(Camera Offset Amount=30)
  ]
  particleUpdateScript [
   &ParticleState()
   &DynamicMaterialParameters(
    Index 0 Param 1=&FloatFromCurve(FloatCurve=DataInterface() {propertyValues {Curve {keys [{arriveTangent=-2.1015, interpMode=RCIM_Cubic, leaveTangent=-2.1015, tangentMode=RCTM_User, time=0, value=0.5}, <at 1 -0.15>]}}})
    Index 0 Param 2=&FloatFromCurve(FloatCurve=DataInterface() {propertyValues {Curve {keys [{arriveTangent=-0.3464, interpMode=RCIM_Cubic, leaveTangent=-0.3464, tangentMode=RCTM_User, time=0, value=1}, <at 1 0.9>]}}})
    Index 0 Param 3=&FloatFromCurve(FloatCurve=DataInterface() {propertyValues {Curve {keys [<at 0 0>,<at 0.15 1>,<at 1 1.2>]}}})
    Module.Param0WriteEnabled=true
    Module.Param1WriteEnabled=true
    Module.Param2WriteEnabled=true
    Module.Param3WriteEnabled=false
   ) {module=DynamicMaterialParameters001}
   &ScaleColor(
    Linear Color Curve=DataInterface() {propertyValues {
      AlphaCurve {keys [<at 0 0.5>,<at 0.967 0>]}
      BlueCurve {keys [<at 0 0.086>,<at 0.6282 0.0167>]}
      GreenCurve {keys [<at 0 0.02>,<at 0.6282 0>]}
      RedCurve {keys [<at 0 2>,<at 0.6282 1>]}
     }}
    Scale Mode=::NewEnumerator2
   )
   &ScaleMeshSize(Scale Factor=&VectorFromFloat(Value=&FloatFromCurve(
      FloatCurve=DataInterface() {propertyValues {Curve {keys [
          {arriveTangent=1.0704, interpMode=RCIM_Cubic, leaveTangent=1.0704, tangentMode=RCTM_User, time=0, value=0.5}
          {interpMode=RCIM_Cubic, tangentMode=RCTM_User, time=1, value=1}
         ]}}}
      Scale Curve=2
     )))
  ]
  rendererClasses [&NiagaraMeshRendererProperties]
  renderers [{rendererClass=&NiagaraMeshRendererProperties, rendererIndex=0}]
 }
 Sparks {bEnabled=false, rendererClasses [], renderers []}
 Sparks003 {
  simTarget=GPUComputeSim
  emitterUpdateScript [
   &EmitterState(
    Life Cycle Mode=::NewEnumerator1
    Loop Behavior=::NewEnumerator1
    Loop Delay=$System.BeamDelaySpawn
    Loop Duration=$System.FX_Duration
    UseLoopDelay=true
   )
   &SpawnRate(Spawn Probability=$System.LaserFadeOut, SpawnRate=40, Use Spawn Probability=true)
  ]
  particleSpawnScript [
   &InitializeParticle(
    Color=<rgba 2 2 2 1>
    Lifetime Max=0.5
    Lifetime Min=0.2
    Lifetime Mode=::NewEnumerator1
    Position Offset:NiagaraPosition=$System.ImpactLocation
    Sprite Size Mode=::NewEnumerator3
    Uniform Sprite Size=&RandomRangeFloat(Minimum=3, Maximum=5)
    UsePositionOffset=true
    Write Material Random=true
   )
   &CameraOffset(Camera Offset Amount=25)
   &ShapeLocation(
    Hemisphere Distribution=<1 0.5>
    MaskSpawnGroup=false
    Override Seed=false
    Sphere Radius=30
    Sphere Surface Distribution=0.8929
    UseOwnerScale=true
   )
   &DynamicMaterialParameters(
    Index 0 Param 1=2
    Index 0 Param 2=10
    Index 0 Param 3=1
    Module.Index0Float0Write=true
    Module.Index0Float1Write=true
    Module.Index0Float2Write=true
    Module.Index0Float3Write=false
   )
   &AddVelocity(
    Module.Use Falloff Distance=false
    Override Seed=false
    Velocity Mode=::NewEnumerator1
    Velocity Origin=$System.ImpactLocation
    Velocity Speed=&RandomRangeFloat(Minimum=300, Maximum=600)
   )
  ]
  particleUpdateScript [
   &ParticleState()
   &AccelerationForce(Acceleration=<0 0 -500>)
   &CurlNoiseForce(
    Bias Noise=false
    Noise Frequency=400
    Noise Strength=&FloatFromCurve(
     FloatCurve=DataInterface() {propertyValues {Curve {keys [{interpMode=RCIM_Cubic, tangentMode=RCTM_User, time=0, value=0}, {interpMode=RCIM_Cubic, tangentMode=RCTM_User, time=0.3, value=1}]}}}
     Scale Curve=5000
    )
    Pan Noise=false
   ) {enabled=false}
   &Drag(Drag=&RandomRangeFloat(Minimum=1, Maximum=2), Use Linear Drag=true, Use Rotational Drag=false)
   &AccelerationForce(Acceleration=<0 0 100>) {enabled=false, module=AccelerationForce001}
   &SolveForcesAndVelocity(Clamp Velocity=false, Limit Acceleration=false)
   &ScaleSpriteSizeBySpeed(Max Scale Factor=<1 4>, Velocity Threshold=600)
   &ScaleSpriteSize(Scale Factor=&Vector2DFromFloat(Value=&FloatFromCurve(
      FloatCurve=DataInterface() {propertyValues {Curve {keys [
          {arriveTangent=7.1771, interpMode=RCIM_Cubic, leaveTangent=7.1771, tangentMode=RCTM_User, time=0, value=0}
          {arriveTangent=0.1164, interpMode=RCIM_Cubic, leaveTangent=0.1164, tangentMode=RCTM_User, time=0.15, value=1}
          {arriveTangent=-0.2479, interpMode=RCIM_Cubic, leaveTangent=-0.2479, tangentMode=RCTM_User, time=0.5, value=1}
          <at 1 0 RCIM_Cubic>
         ]}}}
      Scale Curve=&RandomRangeFloat(Minimum=0.6, Maximum=1.2)
     )))
   &ScaleColor(
    Scale RGB=&VectorFromFloat(Value=&FloatFromCurve(
      FloatCurve=DataInterface() {propertyValues {Curve {keys [
          {interpMode=RCIM_Cubic, tangentMode=RCTM_User, time=0, value=1}
          <at 0.3 0.2>
          {interpMode=RCIM_Cubic, tangentMode=RCTM_User, time=1, value=0}
         ]}}}
      Scale Curve=10
     ))
    ScaleA=true
    ScaleRGB=true
   ) {enabled=false}
   &ScaleColor(
    Scale Alpha=&FloatFromCurve(FloatCurve=DataInterface() {propertyValues {Curve {keys [
         <at 0 0 RCIM_Cubic>
         {arriveTangent=0.159, interpMode=RCIM_Cubic, leaveTangent=0.159, tangentMode=RCTM_User, time=0.1, value=1}
         {interpMode=RCIM_Cubic, tangentMode=RCTM_Break, time=0.85, value=1}
         {interpMode=RCIM_Cubic, tangentMode=RCTM_Break, time=1, value=0}
        ]}}})
    Scale RGB=&MakeVectorFromLinearColorRGB(Color=&ColorFromCurve(ColorCurve=DataInterface() {propertyValues {
        AlphaCurve {keys [<at 0 1 RCIM_Cubic>]}
        BlueCurve {keys [<at 0 0.7 RCIM_Cubic>,<at 0.2653 0.1>,<at 0.9955 0>]}
        GreenCurve {keys [<at 0 0.91 RCIM_Cubic>,<at 0.2653 0.1>,<at 0.9955 0>]}
        RedCurve {keys [<at 0 7 RCIM_Cubic>,<at 0.2653 2>,<at 0.9955 1>]}
       }}))
    ScaleA=false
    ScaleRGB=true
   ) {module=ScaleColor001}
   &FN_Sprite_Attributes(
    Color=&ColorFromCurve(ColorCurve=DataInterface() {propertyValues {
       AlphaCurve {keys [<at 0.0151 0>,<at 0.0946 0.075 RCIM_Cubic>,<at 0.7029 0 RCIM_Cubic>]}
       BlueCurve {keys [<at 0.0598 0 RCIM_Cubic>,<at 0.7123 0.0333>]}
       GreenCurve {keys [<at 0.0598 0.0667 RCIM_Cubic>,<at 0.7123 0>]}
       RedCurve {keys [<at 0.0598 1 RCIM_Cubic>,<at 0.7123 1>]}
      }})
    Create Color Attribute=true
    Create Dynamic Material Parameter Attributes=true
    Create SpriteSize Attribute=true
    Index 0 Param 1=1
    Index 0 Param 2=10
    Index 0 Param 3=1
    Module Output:ENiagara_AttributeSamplingApplyOutput=::NewEnumerator1
    Module.Index0Float0Write=true
    Module.Index0Float1Write=true
    Module.Index0Float2Write=true
    Module.Index0Float3Write=true
    SpriteSize=&Multiply_Vector2D(A=$StackContext.SpriteSize, B=<7 2.6>)
   )
  ]
  rendererClasses [&NiagaraSpriteRendererProperties]
  renderers [
   {rendererClass=&NiagaraSpriteRendererProperties}
   {rendererClass=&NiagaraSpriteRendererProperties, rendererIndex=1}
  ]
 }
}
summary {
 systemName=NS_SouthBulk_Boss_EnergyRay_Impact
 emitters [
  {bEnabled=false, emitterName=ImpactFlashes, rendererClasses [&NiagaraSpriteRendererProperties]}
  {
   bEnabled=false
   emitterName=ImpactCoreOld
   rendererClasses [&NiagaraSpriteRendererProperties, &NiagaraLightRendererProperties]
  }
  {emitterName=ImpactRocks, rendererClasses [&NiagaraMeshRendererProperties]}
  {bEnabled=false, emitterName=Sparks, rendererClasses []}
  {emitterName=ScorchTrail, rendererClasses [&NiagaraRibbonRendererProperties]}
  {emitterName=Shockwave, rendererClasses [&NiagaraMeshRendererProperties]}
  {emitterName=EnergyThorn, rendererClasses [&NiagaraMeshRendererProperties]}
  {emitterName=ImpactCore, rendererClasses [&NiagaraSpriteRendererProperties, &NiagaraLightRendererProperties]}
  {emitterName=Sparks003, simTarget=GPUComputeSim, rendererClasses [&NiagaraSpriteRendererProperties]}
 ]
 userVariables []
}
